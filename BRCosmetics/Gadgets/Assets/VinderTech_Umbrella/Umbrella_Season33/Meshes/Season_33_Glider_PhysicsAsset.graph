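# Physics asset for the Season 33 VinderTech umbrella glider mesh, defining its simplified collision bodies and joints.
# Three capsules cover the handle, the first blade bone and the spinning-top master bone, with the blade capsule by far the widest.
# The blade and spin-top bodies each constrain directly to the handle in a star layout, with matching swing and twist limit angles.
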PhysicsAsset /BRCosmetics/Gadgets/Assets/VinderTech_Umbrella/Umbrella_Season33/Meshes/Season_33_Glider_PhysicsAsset
bodies {
 Melee_TwoHanded_Handle [{shapeType=Capsule, length=64.9287, radius=11.9741, center=<-0.0126 -0.0016 31.4092>, rotation=<rot -0.0022 -90 -0.0179>}]
 blade_01 [{shapeType=Capsule, length=77.5964, radius=77.5964, center=<-0.0012 0.0012 -6.019>, rotation=<rot 0.0021 163.9684 -90>}]
 spin_top_mstr [{shapeType=Capsule, length=25.1604, radius=25.0821, center=<-1.017 -1.1942 8.7031>, rotation=<rot 39.6775 -90 -89.4017>}]
}
constraints [
 {bone1Name=spin_top_mstr, bone2Name=Melee_TwoHanded_Handle, swing1LimitDegrees=45, swing2LimitDegrees=45, twistLimitDegrees=45}
 {bone1Name=blade_01, bone2Name=Melee_TwoHanded_Handle, swing1LimitDegrees=45, swing2LimitDegrees=45, twistLimitDegrees=45}
]
